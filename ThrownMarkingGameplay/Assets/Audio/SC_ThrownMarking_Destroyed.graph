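# Sound cue for a thrown marking being destroyed, assigned to the large-explosives sound class with its own attenuation asset.
# Three equally weighted Random nodes each pick one of three destroy waves, covering nine variations in all.
# A Mixer layers the three picks at different volumes into a root Modulator for the final output.
# A concurrency override caps simultaneous instances.
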SoundCue /ThrownMarkingGameplay/Assets/Audio/SC_ThrownMarking_Destroyed
root=Modulator_1
nodes [
 {id=Mixer_0, children [Random_3,Random_2,Random_4], properties {inputVolume [0.15,0.5,0.25]}}
 {id=Modulator_1, children [Mixer_0]}
 {id=Random_2, children [WavePlayer_7,WavePlayer_8,WavePlayer_9], properties {weights [1,1,1]}}
 {id=Random_3, children [WavePlayer_10, WavePlayer_11, WavePlayer_12], properties {weights [1,1,1]}}
 {id=Random_4, children [WavePlayer_13, WavePlayer_14, WavePlayer_15], properties {weights [1,1,1]}}
 {id=WavePlayer_10, properties {soundWaveAssetPtr=&/ThrownMarking_Destroy_06}}
 {id=WavePlayer_11, properties {soundWaveAssetPtr=&/ThrownMarking_Destroy_05}}
 {id=WavePlayer_12, properties {soundWaveAssetPtr=&/ThrownMarking_Destroy_04}}
 {id=WavePlayer_13, properties {soundWaveAssetPtr=&/ThrownMarking_Destroy_07}}
 {id=WavePlayer_14, properties {soundWaveAssetPtr=&/ThrownMarking_Destroy_08}}
 {id=WavePlayer_15, properties {soundWaveAssetPtr=&/ThrownMarking_Destroy_09}}
 {id=WavePlayer_7, properties {soundWaveAssetPtr=&/ThrownMarking_Destroy_01}}
 {id=WavePlayer_8, properties {soundWaveAssetPtr=&/ThrownMarking_Destroy_02}}
 {id=WavePlayer_9, properties {soundWaveAssetPtr=&/ThrownMarking_Destroy_03}}
]
properties {
 attenuationSettings=&/Att_Throwmarking_Destroyed
 bOverrideConcurrency=true
 maxDistance=12500
 soundClassObject=/Game/Sounds/FortSoundClassesAndModes/Fort_Explosives_Large_PSM
 volumeMultiplier=0.5
 attenuationOverrides {falloffDistance=8000, attenuationShapeExtents {x=0}}
 concurrencyOverrides {maxCount=4}
}
