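# Sound cue for the ThrownMarking equip sound, playing a single assemble wave as player foley.
# The wave player feeds a modulator that applies a fixed volume boost with no pitch or volume randomization.
# It uses the generic player equip attenuation and overrides concurrency to block new instances rather than stop playing ones.
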SoundCue /ThrownMarkingGameplay/Assets/Audio/SC_ThrownMarking_Equip
root=Modulator_0
nodes [
 {id=Modulator_0, children [WavePlayer_9], properties {pitchMax=1, pitchMin=1, volumeMax=1.55, volumeMin=1.55}}
 {id=WavePlayer_9, properties {soundWaveAssetPtr=&/ThrownMarking_Assemble_B}}
]
properties {
 attenuationSettings=/Game/Sounds/Attenuation/Fort_Player_Equip_Generic_Att
 bOverrideConcurrency=true
 maxDistance=1800
 soundClassObject=&Fort_Player_Foley
 volumeMultiplier=1
 concurrencyOverrides {resolutionRule=PreventNew}
}
