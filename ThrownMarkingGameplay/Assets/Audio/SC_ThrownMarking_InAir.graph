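# Sound cue for a thrown marking projectile in flight, mixing a throw sound with a looping warning.
# One branch plays a random, pitched-down throw variant under the gadgets sound class.
# The warning loop, classed as explosives, fades in through an enveloper, with Doppler and its own HRTF, occluded attenuation.
# A PreExplo parameter drives the loop's pitch and volume, and owner-limited explosive fuse concurrency is referenced.
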SoundCue /ThrownMarkingGameplay/Assets/Audio/SC_ThrownMarking_InAir
root=Mixer_0
nodes [
 {
  id=Attenuation_7
  children [ModulatorContinuous_1]
  properties {
   attenuationSettings=&Fort_Weapon_Projectile_Medium_Mono_Athena_Att
   bOverrideAttenuation=true
   attenuationOverrides {
    bEnableOcclusion=true
    bEnableReverbSend=false
    distanceAlgorithm=Logarithmic
    falloffDistance=0
    occlusionLowPassFilterFrequency=1000
    occlusionTraceChannel=ECC_GameTraceChannel11
    spatializationAlgorithm=SPATIALIZATION_HRTF
    stereoSpread=0
    attenuationShapeExtents {x=550}
   }
  }
 }
 {id=Doppler_0, children [Attenuation_7], properties {dopplerIntensity=1.5}}
 {
  id=Enveloper_0
  children [SoundClass_0]
  properties {pitchCurve {editorCurveData {keys [{time=0}]}}, volumeCurve {editorCurveData {keys [<at 0 0>,<at 0.25 1>]}}}
 }
 {id=Mixer_0, children [SoundClass_1,Modulator_2], properties {inputVolume [0.7,1]}}
 {
  id=ModulatorContinuous_1
  children [WavePlayer_13]
  properties {
   pitchModulationParams {default=0, maxOutput=1.8, minOutput=1, parameterName=PreExplo}
   volumeModulationParams {default=0, minOutput=0.1, parameterName=PreExplo}
  }
 }
 {id=Modulator_1, children [Random_2], properties {pitchMax=0.95, pitchMin=0.85}}
 {id=Modulator_2, children [Enveloper_0], properties {pitchMin=1.05}}
 {id=Random_2, children [WavePlayer_9, WavePlayer_10, WavePlayer_11], properties {weights [1,1,1]}}
 {id=SoundClass_0, children [Doppler_0], properties {soundClassOverride=/Game/Sounds/FortSoundClassesAndModes/Fort_Explosives_Medium_PSM}}
 {id=SoundClass_1, children [Modulator_1], properties {soundClassOverride=&Fort_Gadgets}}
 {id=WavePlayer_10, properties {soundWaveAssetPtr=&/ThrownMarking_Throw_02}}
 {id=WavePlayer_11, properties {soundWaveAssetPtr=&/ThrownMarking_Throw_03}}
 {id=WavePlayer_13, properties {bLooping=true, soundWaveAssetPtr=&/Throwmarking_Warning_01}}
 {id=WavePlayer_9, properties {soundWaveAssetPtr=&/ThrownMarking_Throw_01}}
]
properties {
 attenuationSettings=&/Att_Throwmarking_InAirLoop
 maxDistance=22500
 soundClassObject=&Fort_WeaponRanged_Fire_3P_PSM
 volumeMultiplier=0.7
 attenuationOverrides {falloffDistance=5000, attenuationShapeExtents {x=1000}}
 concurrencyOverrides {bLimitToOwner=true, maxCount=3, resolutionRule=StopQuietest}
 concurrencySet [/Game/Sounds/Concurrency_Rules/Athena_Explosive_Fuse_Conc]
}
