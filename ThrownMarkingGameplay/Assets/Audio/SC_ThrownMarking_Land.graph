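# Sound cue for the landing impact of a thrown marking, built from rock impact recordings.
# A Random node picks one of four equally weighted waves and feeds a Modulator at the root for playback variation.
# The cue plays pitched up in the impacts sound class, with small-object hit attenuation.
# A shared concurrency rule, named for a four-voice limit that stops the furthest or oldest, manages overlapping landings.
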SoundCue /ThrownMarkingGameplay/Assets/Audio/SC_ThrownMarking_Land
root=Modulator_1
nodes [
 {id=Modulator_1, children [Random_2]}
 {id=Random_2, children [WavePlayer_19, WavePlayer_20, WavePlayer_21, WavePlayer_0], properties {weights [1,1,1,1]}}
 {id=WavePlayer_0, properties {soundWaveAssetPtr=&/rock_impact_10}}
 {id=WavePlayer_19, properties {soundWaveAssetPtr=&/rock_impact_03}}
 {id=WavePlayer_20, properties {soundWaveAssetPtr=&/rock_impact_06}}
 {id=WavePlayer_21, properties {soundWaveAssetPtr=&/rock_impact_09}}
]
properties {
 attenuationSettings=/Game/Sounds/Attenuation/Fort_Demolition_Hit_Small_Sized_Object_Att
 maxDistance=4000
 pitchMultiplier=1.5
 soundClassObject=&Fort_Impacts
 volumeMultiplier=1
 concurrencyOverrides {maxCount=4}
 concurrencySet [/Game/Sounds/Concurrency_Rules/Demo_FourMaximum_StopFurthestOldest_VolumeScale]
}
