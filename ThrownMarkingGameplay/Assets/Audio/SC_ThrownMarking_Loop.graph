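# Looping SoundCue for the ThrownMarking gameplay feature, used as its positional loop sound.
# A Modulator adds a very slight random pitch rise to a looping Wave Player of the ThrownMarking_Loop wave.
# It uses a dedicated throw-marking attenuation asset and the player foley sound class, with slightly reduced volume.
# Concurrency overrides hold a single instance limited per owner, with a stop-oldest resolution rule.
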SoundCue /ThrownMarkingGameplay/Assets/Audio/SC_ThrownMarking_Loop
root=Modulator_0
nodes [
 {id=Modulator_0, children [WavePlayer_0], properties {pitchMax=1.02, pitchMin=1}}
 {id=WavePlayer_0, properties {bLooping=true, soundWaveAssetPtr=&/ThrownMarking_Loop}}
]
properties {
 attenuationSettings=&/Att_Throwmarking_Loop
 maxDistance=3900
 soundClassObject=&Fort_Player_Foley_Athena
 volumeMultiplier=0.85
 concurrencyOverrides {bLimitToOwner=true, maxCount=1, resolutionRule=StopOldest}
 concurrencySet [null]
}
