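# Sound cue for picking up a ThrownMarking item, using the weapon pickup attenuation and loot-collect sound class.
# A mixer layers two pitch-modulated branches: a random pick of three B variants, pitched down at full mix volume, and the main pickup wave at reduced mix volume.
# A concurrency override allows only one instance at a time and blocks new plays while it is active.
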SoundCue /ThrownMarkingGameplay/Assets/Audio/SC_ThrownMarking_Pickup
root=Mixer_0
nodes [
 {id=Mixer_0, children [Modulator_0,Modulator_1], properties {inputVolume [1,0.5]}}
 {id=Modulator_0, children [Random_1], properties {pitchMax=0.85, pitchMin=0.75}}
 {id=Modulator_1, children [WavePlayer_0], properties {pitchMax=1}}
 {id=Random_1, children [WavePlayer_1,WavePlayer_2,WavePlayer_6], properties {weights [1,1,1]}}
 {id=WavePlayer_0, properties {soundWaveAssetPtr=&/ThrownMarking_Pickup}}
 {id=WavePlayer_1, properties {soundWaveAssetPtr=&/ThrownMarking_Pickup_B_01}}
 {id=WavePlayer_2, properties {soundWaveAssetPtr=&/ThrownMarking_Pickup_B_02}}
 {id=WavePlayer_6, properties {soundWaveAssetPtr=&/ThrownMarking_Pickup_B_03}}
]
properties {
 attenuationSettings=&Fort_Weapon_Pickup_Att
 bOverrideConcurrency=true
 maxDistance=3500
 soundClassObject=&Fort_Loot_Collect
 volumeMultiplier=0.5
 concurrencyOverrides {maxCount=1, resolutionRule=PreventNew}
}
