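# Sound cue for the powerup assemble charge of the thrown marking gameplay.
# It mixes a powerup wave with a quieter assemble wave, then runs both through a modulator that varies pitch and volume on each play.
# It uses the Throwmarking destroyed attenuation, a large-explosives sound class and a borrowed Ostrich vehicle concurrency asset, with stop-oldest, owner-limited settings.
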SoundCue /ThrownMarkingGameplay/Assets/Audio/SC_ThrownMarking_Powerup_Assemble_Charge
root=Modulator_0
nodes [
 {id=Mixer_0, children [WavePlayer_0,WavePlayer_1], properties {inputVolume [0.35,1]}}
 {id=Modulator_0, children [Mixer_0], properties {pitchMin=0.9, volumeMin=0.8}}
 {id=WavePlayer_0, properties {soundWaveAssetPtr=&/ThrownMarking_Powerup}}
 {id=WavePlayer_1, properties {soundWaveAssetPtr=&/ThrownMarking_Assemble_C}}
]
properties {
 attenuationSettings=&/Att_Throwmarking_Destroyed
 maxDistance=12500
 soundClassObject=/Game/Sounds/FortSoundClassesAndModes/Fort_Explosives_Large_PSM
 volumeMultiplier=0.35
 concurrencyOverrides {bLimitToOwner=true, maxCount=1, resolutionRule=StopOldest}
 concurrencySet [/Game/Athena/Sounds/Vehicles/Ostrich/Veh_Ostrich_StopOldestLimitToOwnerFFS_Conc]
}
