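# Sound cue for the prepare-throw foley of the Thrown Marking gadget.
# A mixer layers two throw foley waves at reduced levels, one passing through a modulator for slight random volume variation.
# The quality-level root routes all four of its inputs to that same mixer.
# The cue uses gadget-placement attenuation and the gadget items sound class, with a concurrency override that prevents new instances at the limit.
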SoundCue /ThrownMarkingGameplay/Assets/Audio/SC_ThrownMarking_PrepareThrow
root=QualityLevel_0
nodes [
 {id=Mixer_0, children [WavePlayer_4,Modulator_0], properties {inputVolume [0.5,0.75]}}
 {id=Modulator_0, children [WavePlayer_11], properties {volumeMax=0.6, volumeMin=0.5}}
 {id=QualityLevel_0, children [Mixer_0,Mixer_0,Mixer_0,Mixer_0]}
 {id=WavePlayer_11, properties {soundWaveAssetPtr=&/ThrownMarking_ThrowFoley_B_04}}
 {id=WavePlayer_4, properties {soundWaveAssetPtr=&/ThrownMarking_ThrowFoley_01}}
]
properties {
 attenuationSettings=/Game/Sounds/Attenuation/Fort_GadgetPlace_Att
 bOverrideConcurrency=true
 maxDistance=2000
 soundClassObject=&Fort_Gadgets_Items
 volumeMultiplier=0.35
 concurrencyOverrides {resolutionRule=PreventNew}
}
