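# Sound cue for throwing the Thrown Marking item, classed under gadget items with gadget-placement attenuation.
# A mixer layers a fixed foley wave with two random picks, one from three throw sounds and one from four alternate foley sounds.
# A modulator randomizes pitch in a lowered range before a quality node, which feeds three levels and leaves the fourth empty.
# A concurrency override prevents new instances once the limit is reached.
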SoundCue /ThrownMarkingGameplay/Assets/Audio/SC_ThrownMarking_Throw
root=QualityLevel_0
nodes [
 {id=Mixer_0, children [WavePlayer_4,Random_1,Random_0], properties {inputVolume [0.4,0.5,0.75]}}
 {id=Modulator_0, children [Mixer_0], properties {pitchMax=0.9, pitchMin=0.8}}
 {id=QualityLevel_0, children [Modulator_0,Modulator_0,Modulator_0,null]}
 {id=Random_0, children [WavePlayer_8, WavePlayer_9, WavePlayer_10, WavePlayer_11], properties {weights [1,1,1,1]}}
 {id=Random_1, children [WavePlayer_5,WavePlayer_6,WavePlayer_7], properties {weights [1,1,1]}}
 {id=WavePlayer_10, properties {soundWaveAssetPtr=&/ThrownMarking_ThrowFoley_B_03}}
 {id=WavePlayer_11, properties {soundWaveAssetPtr=&/ThrownMarking_ThrowFoley_B_04}}
 {id=WavePlayer_4, properties {soundWaveAssetPtr=&/ThrownMarking_ThrowFoley_01}}
 {id=WavePlayer_5, properties {soundWaveAssetPtr=&/ThrownMarking_Throw_03}}
 {id=WavePlayer_6, properties {soundWaveAssetPtr=&/ThrownMarking_Throw_02}}
 {id=WavePlayer_7, properties {soundWaveAssetPtr=&/ThrownMarking_Throw_01}}
 {id=WavePlayer_8, properties {soundWaveAssetPtr=&/ThrownMarking_ThrowFoley_B_01}}
 {id=WavePlayer_9, properties {soundWaveAssetPtr=&/ThrownMarking_ThrowFoley_B_02}}
]
properties {
 attenuationSettings=/Game/Sounds/Attenuation/Fort_GadgetPlace_Att
 bOverrideConcurrency=true
 maxDistance=2000
 soundClassObject=&Fort_Gadgets_Items
 volumeMultiplier=0.5
 concurrencyOverrides {resolutionRule=PreventNew}
}
