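# Sound cue for the impact of a thrown marking, routed to the impacts sound class.
# A mixer layers the main impact wave over a quieter sub-mix of two medium landing waves played together, which a modulator pitches well up.
# An outer modulator adds slight random pitch variation to the whole mix.
# Attenuation is overridden, and a shared concurrency rule caps simultaneous plays.
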SoundCue /ThrownMarkingGameplay/Assets/Audio/SC_ThrownMarking_Throw_Impact
root=Modulator_0
nodes [
 {id=Mixer_1, children [WavePlayer_8,WavePlayer_9], properties {inputVolume [1,1]}}
 {id=Mixer_2, children [WavePlayer_3,Modulator_1], properties {inputVolume [1,0.4]}}
 {id=Modulator_0, children [Mixer_2], properties {pitchMax=1.02, pitchMin=0.98}}
 {id=Modulator_1, children [Mixer_1], properties {pitchMax=1.35, pitchMin=1.3}}
 {id=WavePlayer_3, properties {soundWaveAssetPtr=&/ThrownMarking_Throw_Impact_01}}
 {id=WavePlayer_8, properties {soundWaveAssetPtr=&/ThrownMarking_Land_Med_01}}
 {id=WavePlayer_9, properties {soundWaveAssetPtr=&/ThrownMarking_Land_Med_02}}
]
properties {
 attenuationSettings=/Game/Sounds/Attenuation/Fort_Demolition_Hit_Small_Sized_Object_Att
 bOverrideAttenuation=true
 maxDistance=2600
 soundClassObject=&Fort_Impacts
 attenuationOverrides {falloffDistance=2400, stereoSpread=2000, attenuationShapeExtents {x=200}}
 concurrencyOverrides {maxCount=4}
 concurrencySet [/Game/Sounds/Concurrency_Rules/Demo_FourMaximum_StopFurthestOldest_VolumeScale]
}
