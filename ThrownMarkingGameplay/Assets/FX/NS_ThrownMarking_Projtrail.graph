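# Niagara projectile trail for the thrown marking gameplay, combining a glowing flare with a warm ribbon trail.
# Flare bursts a looping-lifetime particle that feeds four curve-pulsed sprite layers and a light through per-renderer attribute modules.
# Trail spawns ribbon particles per distance moved, offsets them against owner velocity, and fades their color and narrows width over life.
# System Timeline's inverse normalized time dims both emitters together.
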NiagaraSystem /ThrownMarkingGameplay/Assets/FX/NS_ThrownMarking_Projtrail
dependencies {
 usedDataInterfaces [&NiagaraDataInterfaceCurve, &NiagaraDataInterfaceColorCurve]
 usedDynamicInputs []
 usedModules [&SystemState, &Timeline]
}
emitters {
 Flare {
  emitterUpdateScript [
   &EmitterState(
    Inactive Response=::NewEnumerator1
    Life Cycle Mode=::NewEnumerator1
    Loop Behavior=::NewEnumerator1
    Loop Duration Mode=::NewEnumerator1
    UseLoopDelay=false
   )
   &SpawnBurst_Instantaneous(Use Spawn Probability=false, UseLoopCountLimit=false)
  ]
  particleSpawnScript [
   /Niagara/Modules/Spawn/Initialization/InitializeParticle(
    Lifetime=2
    Module.Write Color=true
    Module.Write Lifetime=true
    Module.Write Mass=true
    Module.Write Material Random=true
    Module.Write Position=false
    Module.Write Scale=false
    Module.Write SpriteRotation=true
    Module.Write SpriteSize=true
    Sprite Rotation=&RandomRangeFloat(Minimum=0, Maximum=0)
    Sprite Size=&Multiply_Vector2D_ByFloat(Float=3500, Vector2D=<1 1>)
   )
   &Color(Color=<rgba 0 0.0949 0.1424 0.6>)
  ]
  particleUpdateScript [
   &ParticleState(Kill Particles When Lifetime Has Elapsed=false, Loop Particles Lifetime=true)
   &ScaleColor(ScaleA=true, ScaleRGB=true)
   /SetVariables(Particles.Size:NiagaraFloat=&FloatFromCurve(FloatCurve=DataInterface() {propertyValues {Curve {keys [
         {interpMode=RCIM_Cubic, tangentMode=RCTM_User, time=0, value=0.8}
         {interpMode=RCIM_Cubic, tangentMode=RCTM_User, time=0.5, value=1}
         {tangentMode=RCTM_User, time=1, value=0.8}
        ]}}})) {isSetParameters=true}
   &FN_Sprite_Attributes(
    Camera Offset=-20
    Color=&Multiply_LinearColorByFloat(Linear Color=<rgba 0.2 0.1321 0.0535 1>, Float=$System.Timeline.InverseNormalizedTime)
    Create Camera Offset Attribute=true
    Create Color Attribute=true
    Create SpriteSize Attribute=true
    SpriteSize=&Vector2DFromFloat(Value=&Multiply_Float(A=200, B=$Particles.Size))
    Write to Module Namespace=true
   )
   &FN_Sprite_Attributes(
    Color=&Multiply_LinearColorByFloat(Linear Color=<rgba 0.5 0.4089 0.1052 1>, Float=$System.Timeline.InverseNormalizedTime)
    Create Camera Offset Attribute=true
    Create Color Attribute=true
    Create SpriteSize Attribute=true
    SpriteSize=&Vector2DFromFloat(Value=&Multiply_Float(A=65, B=$Particles.Size))
    Write to Module Namespace=true
   ) {module=FN_Sprite_Attributes001}
   &FN_Sprite_Attributes(
    Camera Offset=200
    Color=&MakeLinearColorFromVectorAndFloat(
     "Vector (RGB)"=&MakeVectorFromLinearColorRGB(Color=<rgba 0.5 0.4208 0.1428 1>)
     "Float (Alpha)"=&FloatFromCurve(
      FloatCurve=DataInterface() {propertyValues {Curve {keys [
          {interpMode=RCIM_Cubic, tangentMode=RCTM_User, time=0, value=0.1}
          {interpMode=RCIM_Cubic, tangentMode=RCTM_User, time=0.5, value=0.3}
          {tangentMode=RCTM_User, time=1, value=0.1}
         ]}}}
      Scale Curve=$System.Timeline.InverseNormalizedTime
     )
    )
    Create Camera Offset Attribute=true
    Create Color Attribute=true
    Create SpriteSize Attribute=true
    SpriteSize=&Vector2DFromFloat(Value=&Multiply_Float(
      A=80
      B=&FloatFromCurve(FloatCurve=DataInterface() {propertyValues {Curve {keys [
           {interpMode=RCIM_Cubic, tangentMode=RCTM_User, time=0, value=0.5}
           {interpMode=RCIM_Cubic, tangentMode=RCTM_User, time=0.5, value=0.75}
           {tangentMode=RCTM_User, time=1, value=0.5}
          ]}}})
     ))
    Write to Module Namespace=true
   ) {module=FN_Sprite_Attributes002}
   &FN_Sprite_Attributes(
    Camera Offset=45
    Color=&MakeLinearColorFromVectorAndFloat(
     "Vector (RGB)"=&MakeVectorFromLinearColorRGB(Color=<rgba 1 0.8332 0 1>)
     "Float (Alpha)"=&Multiply_Float(A=&ClampFloat(Float=0.5), B=$System.Timeline.InverseNormalizedTime)
    )
    Create Camera Offset Attribute=true
    Create Color Attribute=true
    Create SpriteSize Attribute=true
    SpriteSize=&Multiply_Vector2(
     A=&RandomRangeVector2D(Minimum=<100 50>, Maximum=<150 70>, Evaluation Type=::NewEnumerator1)
     B=&Vector2DFromFloat(Value=&FloatFromCurve(
       FloatCurve=DataInterface() {propertyValues {Curve {keys [
           {interpMode=RCIM_Cubic, tangentMode=RCTM_User, time=0, value=0.5}
           {interpMode=RCIM_Cubic, tangentMode=RCTM_User, time=0.5, value=1}
           {tangentMode=RCTM_User, time=1, value=0.5}
          ]}}}
       Scale Curve=2
      ))
    )
    Write to Module Namespace=true
   ) {module=FN_Sprite_Attributes004}
   &FN_Light_Attributes(
    Color=<rgba 1 0.6113 0 1>
    Create Color Attribute=true
    Create Exponent Attribute=true
    Create Radius Attribute=true
    Exponent=2
    Radius=&Multiply_Float(A=15, B=$System.Timeline.InverseNormalizedTime)
    Scale Color=&Multiply_Float(A=5, B=$System.Timeline.InverseNormalizedTime)
   )
  ]
  rendererClasses [&NiagaraSpriteRendererProperties, &NiagaraLightRendererProperties]
  renderers [
   {rendererClass=&NiagaraSpriteRendererProperties}
   {rendererClass=&NiagaraSpriteRendererProperties, rendererIndex=1}
   {rendererClass=&NiagaraSpriteRendererProperties, rendererIndex=2}
   {rendererClass=&NiagaraSpriteRendererProperties, rendererIndex=3}
   {rendererClass=&NiagaraLightRendererProperties, rendererIndex=4}
  ]
 }
 Trail {
  emitterUpdateScript [
   &EmitterState()
   &SpawnPerUnit(Max Movement Threshold=100, Module.Use Max Movement Threshold=false, Module.Use Movement Tolerance=false, Use Spawn Probability=false)
  ]
  particleSpawnScript [&InitializeRibbon(Lifetime=0.3, Ribbon Width=10, UsePositionOffset=false, Write Material Random=true)]
  particleUpdateScript [
   &ParticleState()
   &ScaleColor(
    Scale Mode=::NewEnumerator1
    Scale RGBA=/Niagara/DynamicInputs/Multiply/Multiply_Vector4_ByFloat(
     Vector4:Vector4f=&MakeVector4_FromLinearColor(LinearColor=&ColorFromCurve(ColorCurve=DataInterface() {propertyValues {
         AlphaCurve {keys [<at 0 0.5>,<at 1 0>]}
         BlueCurve {keys [<at 0 0.4427>,<at 1 0.3588>]}
         GreenCurve {keys [<at 0 2.9557>,<at 1 0.5388>]}
         RedCurve {keys [<at 0 5>,<at 1 0.599>]}
        }}))
     Float:NiagaraFloat=$System.Timeline.InverseNormalizedTime
    )
    ScaleRGBA=true
   )
   &DynamicMaterialParameters(
    Index 0 Param 1=&FloatFromCurve(FloatCurve=DataInterface() {propertyValues {Curve {keys [{interpMode=RCIM_Cubic, tangentMode=RCTM_User, time=0, value=1}, {interpMode=RCIM_Cubic, tangentMode=RCTM_User, time=1, value=0}]}}})
    Module.Param0WriteEnabled=true
    Module.Param1WriteEnabled=true
    Module.Param2WriteEnabled=true
    Module.Param3WriteEnabled=true
   )
   &ScaleRibbonWidth(Ribbon Width Scale=&FloatFromCurve(FloatCurve=DataInterface() {propertyValues {Curve {keys [
         {arriveTangent=6.8194, interpMode=RCIM_Cubic, leaveTangent=6.8194, tangentMode=RCTM_User, time=0, value=0.8}
         {arriveTangent=-0.3, interpMode=RCIM_Cubic, leaveTangent=-0.3, time=0.1, value=1}
         {interpMode=RCIM_Cubic, tangentMode=RCTM_User, time=1, value=0.5}
        ]}}}))
   &InheritSourceMovement(
    Applied Position Scale=&VectorFromFloat(Value=&FloatFromCurve(FloatCurve=DataInterface() {propertyValues {Curve {keys [<at 0 1>,<at 1 2>]}}}, Scale Curve=-1))
    ApplyForce=false
    ApplyPosition=true
    ApplyPositionFalloff=false
    Source Velocity:Vector3f=&TransformVector(Vector=$Engine.Owner.Velocity, Source Space=::World, Destination Space=::Local)
   )
   &SolveForcesAndVelocity(Clamp Velocity=false, Limit Acceleration=false)
  ]
  rendererClasses [&NiagaraRibbonRendererProperties]
  renderers [{rendererClass=&NiagaraRibbonRendererProperties}]
 }
}
summary {
 systemName=NS_ThrownMarking_Projtrail
 emitters [
  {emitterName=Flare, rendererClasses [&NiagaraSpriteRendererProperties, &NiagaraLightRendererProperties]}
  {emitterName=Trail, rendererClasses [&NiagaraRibbonRendererProperties]}
 ]
 userVariables [
  User.EnemyRecon {type=NiagaraFloat, defaultValue=0, description=""}
  User.FadeOut {type=NiagaraBool}
 ]
}
