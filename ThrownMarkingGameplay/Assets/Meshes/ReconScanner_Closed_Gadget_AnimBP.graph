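# Animation Blueprint for the closed Recon Scanner gadget mesh, adding secondary physics sway to its body bone.
# The input pose runs through a root-relative Anim Dynamics node with an angular spring, damped component motion, custom gravity and rotation limited to the Y axis.
# An additive Modify Bone node then applies a small fixed roll offset to the same bone before output.
# The event graph holds only unconnected stub nodes.
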AnimBlueprint /ThrownMarkingGameplay/Assets/Meshes/ReconScanner_Closed_Gadget_AnimBP
target_skeleton=&/ReconScanner_Closed_Gadget_Skeleton
graphs [
 AnimGraph {nodes [
   n1 = Anim Dynamics - Bone: body(ComponentPose=Local To Component(LocalPose=Input Pose()), Alpha=1, ExternalForce="0,0,0") {
    angularDampingOverride=0.5
    angularSpringConstant=0.9
    bAngularSpring=true
    bOverrideAngularDamping=true
    bUseGravityOverride=true
    gravityScale=0.35
    simulationSpace=RootRelative
    alphaBoolBlend {blendOption=Linear}
    boundBone {boneName=body}
    componentLinearAccScale=<0.4 0.4 0.4>
    componentLinearVelScale=<0.4 0.4 0.4>
    gravityOverride {z=-1500}
    physicsBodyDefinitions [{
      boundBone {boneName=body}
      boxExtents=<35 35 35>
      constraintSetup {angularLimitsMax=<0 30 0>, angularLimitsMin=<0 -30 0>, angularTarget=<1 0 0>}
      localJointOffset=<5 0 0>
     }]
   }
   Output Pose(Result=Component To Local(ComponentPose="Transform (Modify) Bone - Bone: body"(ComponentPose=@n1, Alpha=1, Translation="0,0,0", Rotation="0,0,-8", Scale="1,1,1") {rotationMode=BMM_Additive, alphaBoolBlend {blendOption=Linear}, boneToModify {boneName=body}, rotation {roll=-8}}))
  ]}
 EventGraph {nodes [
   TryGetPawnOwner()
   Event BlueprintUpdateAnimation()
  ]}
]
variables []
